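# Post-style animation blueprint for the Jada female head that adds secondary jiggle motion to its hair bones on the Fortnite base head skeleton.
# The input pose runs through a series of ten single-body AnimDynamics nodes, one for dyn_hair_01 and each of its lettered sub-bones, then returns to local space.
# Each uses linear and angular springs in actor space with zero gravity, tight limits and low alpha for subtle sway.
# The event graph holds no logic.
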AnimBlueprint /BRCosmetics/Characters/Player/Female/Medium/Heads/F_MED_BLK_Jada_Head_01/Meshes/F_MED_BLK_Jada_Head_01_Dyn_AnimBP
target_skeleton=/Game/Characters/Player/Common/Fortnite_Base_Head/Fortnite_Base_Head_Export_Skeleton
graphs [
 AnimGraph {nodes [
   AnimGraphNode_SubInput_0 = Input Pose()
   n1 = Anim Dynamics - Bone: dyn_hair_01(ComponentPose=Local To Component(LocalPose=@AnimGraphNode_SubInput_0), Alpha=0.126) {
    angularSpringConstant=20
    bAngularSpring=true
    bGravityOverrideInSimSpace=true
    bLinearSpring=true
    gravityScale=0
    lODThreshold=1
    linearSpringConstant=20
    simulationSpace=Actor
    alphaBoolBlend {blendOption=Linear}
    boundBone {boneName=dyn_hair_01}
    physicsBodyDefinitions [{
      sphereCollisionRadius=0
      boundBone {boneName=dyn_hair_01}
      boxExtents=<30 30 30>
      constraintSetup {
       twistAxis=AxisZ
       angularLimitsMax=<3 7.5 0>
       angularLimitsMin=<-3 -7.5 0>
       angularTarget=<0 0 0>
       linearAxesMax=<1.5 6 5>
       linearAxesMin=<-1.5 -6 -0.5>
      }
      localJointOffset=<0 0 0>
     }]
    relativeSpaceBone {boneName=dyn_hair_01}
   }
   n2 = Anim Dynamics - Bone: dyn_hair_01_a(ComponentPose=@n1, Alpha=0.2, ExternalForce="0,0,0") {
    angularSpringConstant=65
    bAngularSpring=true
    bGravityOverrideInSimSpace=true
    bLinearSpring=true
    gravityScale=0
    lODThreshold=1
    linearSpringConstant=65
    simulationSpace=Actor
    alphaBoolBlend {blendOption=Linear}
    boundBone {boneName=dyn_hair_01_a}
    physicsBodyDefinitions [{
      sphereCollisionRadius=0
      boundBone {boneName=dyn_hair_01_a}
      boxExtents=<30 30 30>
      constraintSetup {
       twistAxis=AxisZ
       angularLimitsMax=<3 15 0>
       angularLimitsMin=<-3 -15 0>
       angularTarget=<0 0 0>
       linearAxesMax=<1.5 1 5>
       linearAxesMin=<-1.5 -1 -0.5>
      }
      localJointOffset=<0 0 0>
     }]
    relativeSpaceBone {boneName=dyn_hair_01_a}
   }
   n3 = Anim Dynamics - Bone: dyn_hair_01_b(ComponentPose=@n2, Alpha=0.23, ExternalForce="0,0,0") {
    angularSpringConstant=60
    bAngularSpring=true
    bGravityOverrideInSimSpace=true
    bLinearSpring=true
    gravityScale=0
    lODThreshold=1
    linearSpringConstant=60
    simulationSpace=Actor
    alphaBoolBlend {blendOption=Linear}
    boundBone {boneName=dyn_hair_01_b}
    physicsBodyDefinitions [{
      sphereCollisionRadius=0
      boundBone {boneName=dyn_hair_01_b}
      boxExtents=<30 30 30>
      constraintSetup {
       twistAxis=AxisZ
       angularLimitsMax=<3 7.5 5>
       angularLimitsMin=<-3 -7.5 -5>
       angularTarget=<0 0 0>
       linearAxesMax=<3 3 5>
       linearAxesMin=<-3 -3 -0.5>
      }
      localJointOffset=<0 0 0>
     }]
    relativeSpaceBone {boneName=dyn_hair_01_b}
   }
   n4 = Anim Dynamics - Bone: dyn_hair_01_c(ComponentPose=@n3, Alpha=0.26, ExternalForce="0,0,0") {
    angularSpringConstant=80
    bAngularSpring=true
    bGravityOverrideInSimSpace=true
    bLinearSpring=true
    gravityScale=0
    lODThreshold=1
    linearSpringConstant=80
    simulationSpace=Actor
    alphaBoolBlend {blendOption=Linear}
    boundBone {boneName=dyn_hair_01_c}
    physicsBodyDefinitions [{
      sphereCollisionRadius=0
      boundBone {boneName=dyn_hair_01_c}
      boxExtents=<30 30 30>
      constraintSetup {
       twistAxis=AxisZ
       angularLimitsMax=<3 7.5 0>
       angularLimitsMin=<-3 -7.5 0>
       angularTarget=<0 0 0>
       linearAxesMax=<3 1.5 2.5>
       linearAxesMin=<-1.5 -0.5 -0.5>
      }
      localJointOffset=<0 0 0>
     }]
    relativeSpaceBone {boneName=dyn_hair_01_c}
   }
   n5 = Anim Dynamics - Bone: dyn_hair_01_d(ComponentPose=@n4, Alpha=0.175, ExternalForce="0,0,0") {
    angularSpringConstant=70
    bAngularSpring=true
    bGravityOverrideInSimSpace=true
    bLinearSpring=true
    gravityScale=0
    lODThreshold=1
    linearSpringConstant=70
    simulationSpace=Actor
    alphaBoolBlend {blendOption=Linear}
    boundBone {boneName=dyn_hair_01_d}
    physicsBodyDefinitions [{
      sphereCollisionRadius=0
      boundBone {boneName=dyn_hair_01_d}
      boxExtents=<30 30 30>
      constraintSetup {
       twistAxis=AxisZ
       angularLimitsMax=<6 15 15>
       angularLimitsMin=<-3 -15 -15>
       angularTarget=<0 0 0>
       linearAxesMax=<3 1 5>
       linearAxesMin=<-4 -2 -0.5>
      }
      localJointOffset=<0 0 0>
     }]
    relativeSpaceBone {boneName=dyn_hair_01_d}
   }
   n6 = Anim Dynamics - Bone: dyn_hair_01_e(ComponentPose=@n5, Alpha=0.21, ExternalForce="0,0,0") {
    angularSpringConstant=75
    bAngularSpring=true
    bGravityOverrideInSimSpace=true
    bLinearSpring=true
    gravityScale=0
    lODThreshold=1
    linearSpringConstant=75
    simulationSpace=Actor
    alphaBoolBlend {blendOption=Linear}
    boundBone {boneName=dyn_hair_01_e}
    physicsBodyDefinitions [{
      sphereCollisionRadius=0
      boundBone {boneName=dyn_hair_01_e}
      boxExtents=<30 30 30>
      constraintSetup {
       twistAxis=AxisZ
       angularLimitsMax=<3 15 0>
       angularLimitsMin=<-3 -15 0>
       angularTarget=<0 0 0>
       linearAxesMax=<4 6 5>
       linearAxesMin=<-1.5 -6 -0.5>
      }
      localJointOffset=<0 0 0>
     }]
    relativeSpaceBone {boneName=dyn_hair_01_e}
   }
   n7 = Anim Dynamics - Bone: dyn_hair_01_f(ComponentPose=@n6, Alpha=0.26, ExternalForce="0,0,0") {
    angularSpringConstant=80
    bAngularSpring=true
    bGravityOverrideInSimSpace=true
    bLinearSpring=true
    gravityScale=0
    lODThreshold=1
    linearSpringConstant=80
    simulationSpace=Actor
    alphaBoolBlend {blendOption=Linear}
    boundBone {boneName=dyn_hair_01_f}
    physicsBodyDefinitions [{
      sphereCollisionRadius=0
      boundBone {boneName=dyn_hair_01_f}
      boxExtents=<30 30 30>
      constraintSetup {
       twistAxis=AxisZ
       angularLimitsMax=<10 7.5 10>
       angularLimitsMin=<-10 -7.5 -10>
       angularTarget=<0 0 0>
       linearAxesMax=<1.5 2 5>
       linearAxesMin=<-3 -2 -0.5>
      }
      localJointOffset=<0 0 0>
     }]
    relativeSpaceBone {boneName=dyn_hair_01_f}
   }
   n8 = Anim Dynamics - Bone: dyn_hair_01_g(ComponentPose=@n7, Alpha=0.22, ExternalForce="0,0,0") {
    angularSpringConstant=50
    bAngularSpring=true
    bGravityOverrideInSimSpace=true
    bLinearSpring=true
    gravityScale=0
    lODThreshold=1
    linearSpringConstant=50
    simulationSpace=Actor
    alphaBoolBlend {blendOption=Linear}
    boundBone {boneName=dyn_hair_01_g}
    physicsBodyDefinitions [{
      sphereCollisionRadius=0
      boundBone {boneName=dyn_hair_01_g}
      boxExtents=<30 30 30>
      constraintSetup {
       twistAxis=AxisZ
       angularLimitsMax=<3 7.5 10>
       angularLimitsMin=<-3 -7.5 -10>
       angularTarget=<0 0 0>
       linearAxesMax=<5 2 5>
       linearAxesMin=<-5 -2 -0.5>
      }
      localJointOffset=<0 0 0>
     }]
    relativeSpaceBone {boneName=dyn_hair_01_g}
   }
   n9 = Anim Dynamics - Bone: dyn_hair_01_h(ComponentPose=@n8, Alpha=0.25, ExternalForce="0,0,0") {
    angularSpringConstant=45
    bAngularSpring=true
    bGravityOverrideInSimSpace=true
    bLinearSpring=true
    gravityScale=0
    lODThreshold=1
    linearSpringConstant=45
    simulationSpace=Actor
    alphaBoolBlend {blendOption=Linear}
    boundBone {boneName=dyn_hair_01_h}
    physicsBodyDefinitions [{
      sphereCollisionRadius=0
      boundBone {boneName=dyn_hair_01_h}
      boxExtents=<30 30 30>
      constraintSetup {
       twistAxis=AxisZ
       angularLimitsMax=<3 3 15>
       angularLimitsMin=<-3 -3 -15>
       angularTarget=<0 0 0>
       linearAxesMax=<3 1 5>
       linearAxesMin=<-3 -3 -0.5>
      }
      localJointOffset=<0 0 0>
     }]
    relativeSpaceBone {boneName=dyn_hair_01_h}
   }
   n10 = Anim Dynamics - Bone: dyn_hair_01_i(ComponentPose=@n9, Alpha=0.08, ExternalForce="0,0,0") {
    angularSpringConstant=80
    bAngularSpring=true
    bGravityOverrideInSimSpace=true
    bLinearSpring=true
    gravityScale=0
    lODThreshold=1
    linearSpringConstant=80
    simulationSpace=Actor
    alphaBoolBlend {blendOption=Linear}
    boundBone {boneName=dyn_hair_01_i}
    physicsBodyDefinitions [{
      sphereCollisionRadius=0
      boundBone {boneName=dyn_hair_01_i}
      boxExtents=<30 30 30>
      constraintSetup {
       twistAxis=AxisZ
       angularLimitsMax=<3 7.5 0>
       angularLimitsMin=<-3 -7.5 0>
       angularTarget=<0 0 0>
       linearAxesMax=<1.5 2 5>
       linearAxesMin=<-4 -2 -0.5>
      }
      localJointOffset=<0 0 0>
     }]
    relativeSpaceBone {boneName=dyn_hair_01_i}
   }
   Output Pose(Result=Component To Local(ComponentPose=@n10)) {layerGroup=None}
  ]}
 EventGraph {nodes [Event BlueprintUpdateAnimation()]}
]
variables []
