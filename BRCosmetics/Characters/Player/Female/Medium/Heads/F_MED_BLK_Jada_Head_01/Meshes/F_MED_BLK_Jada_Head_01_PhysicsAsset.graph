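# Physics asset for the Jada female medium player head cosmetic (F_MED_BLK_Jada_Head_01).
# It defines three capsule bodies, on the spine_05, head and faceAttach bones, with the head capsule the largest.
# Two constraints chain faceAttach to head and head to spine_05, using the same angle limit on both swing axes and twist.
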PhysicsAsset /BRCosmetics/Characters/Player/Female/Medium/Heads/F_MED_BLK_Jada_Head_01/Meshes/F_MED_BLK_Jada_Head_01_PhysicsAsset
bodies {
 faceAttach [{shapeType=Capsule, length=7.6377, radius=6.4576, center=<0.8832 5.6644 2.9557>, rotation=<rot -0.1321 -90 -5.968>}]
 head [{shapeType=Capsule, length=14.6938, radius=13.5897, center=<6.9545 -0.1269 0.2877>, rotation=<rot 0 161.3578 -86.9019>}]
 spine_05 [{shapeType=Capsule, length=8.0522, radius=7.4719, center=<10.7847 0.9459 -0.0103>, rotation=<rot 0 50.5237 -88.4816>}]
}
constraints [
 {bone1Name=head, bone2Name=spine_05, swing1LimitDegrees=45, swing2LimitDegrees=45, twistLimitDegrees=45}
 {bone1Name=faceAttach, bone2Name=head, swing1LimitDegrees=45, swing2LimitDegrees=45, twistLimitDegrees=45}
]
